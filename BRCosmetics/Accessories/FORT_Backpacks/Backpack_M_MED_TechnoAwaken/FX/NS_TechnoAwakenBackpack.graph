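# Niagara effect for the TechnoAwaken back bling: layered glowing mesh rings, a 3x3 waveform-pulsed cube grid, glows and an audio-amplitude-driven light.
# System ReactiveValue and PickaxeR values scale size, brightness and material animation speed through curves and lerps.
# BurstRing001 continuously emits expanding rings, while BurstRing002 and 003 flash on a system DoOnce trigger.
# Timeline modules feed animation time to materials; front-end, time-of-day and platform checks adjust the look.
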NiagaraSystem /BRCosmetics/Accessories/FORT_Backpacks/Backpack_M_MED_TechnoAwaken/FX/NS_TechnoAwakenBackpack
dependencies {
 usedDataInterfaces [
  &NiagaraDataInterfaceArrayColor
  &NiagaraDataInterfaceArrayFloat3
  &NiagaraDataInterfaceArrayFloat
  &NiagaraDataInterfaceArrayInt32
  &NiagaraDataInterfacePlatformSet
  &NiagaraDataInterfaceCurve
 ]
 usedDynamicInputs [&InvertBool, &ClampFloat, &ModuloFloat]
 usedModules [
  &SystemState
  &self:SystemSpawnScript.SystemScriptSource.SystemScriptGraph.NiagaraNodeAssignment_0.SetVariables_#1
  &self:SystemSpawnScript.SystemScriptSource.SystemScriptGraph.NiagaraNodeAssignment_1.SetVariables_#2
  &DoOnce
 ]
}
emitters {
 BurstRing {
  emitterUpdateScript [
   &EmitterState(
    Inactive Response=::NewEnumerator1
    Life Cycle Mode=::NewEnumerator1
    Loop Behavior=::NewEnumerator1
    Loop Duration Mode=::NewEnumerator1
    UseLoopDelay=false
   )
   &SpawnBurst_Instantaneous(Use Spawn Probability=false, UseLoopCountLimit=false)
  ]
  particleSpawnScript [
   &InitializeParticle(
    Color=<rgba 0.4167 0 5 1>
    Lifetime=&Multiply_Float(A=1, B=&FN_FrontEndFloatModifier(Gameplay=1, Front End=0.7))
    Mesh Scale Mode=::NewEnumerator3
    Mesh Uniform Scale=0.055
    Position Offset=<0.6 -9.87 0.5>
    Sprite Size Mode=::NewEnumerator3
    UsePositionOffset=true
    Write Material Random=true
   )
   &Color(Color=&MultiplyLinearColor(
     Linear Color=$Particles.Color
     Scale Factor=&MakeVector4(
      X=1
      Y=1
      Z=1
      W=&MakeCustomFloatFromBool(
       Bool=&SetBoolByFloatComparison(A=&Subtract_Float(A=$Emitter.Age, B=$Emitter.float), B=-0.01)
       True Float=0.2
       False Float=1
      )
     )
    ))
   /SetVariables(Particles.bool:NiagaraBool=&SetBoolByFloatComparison(A=&Subtract_Float(A=$Emitter.Age, B=$Emitter.float), B=-0.01)) {isSetParameters=true, module=SetVariables_#3}
  ]
  particleUpdateScript [
   &ParticleState(Kill Particles When Lifetime Has Elapsed=false, Let Infinitely Lived Particles Die When Emitter Deactivates=true)
   &Timeline(
    Play Rate=&FloatFromCurve(FloatCurve=DataInterface() {propertyValues {Curve {keys [<at 0 1>,<at 0.9 2>,<at 1 1>]}}}, CurveIndex=$System.ReactiveValue)
    Timeline End=100000
   )
   &CameraOffset()
   &DynamicMaterialParameters(
    Index 0 Param 1=&Multiply_Float(A=&MakeFloatFromVector(Vector=$Particles.Scale, Channel=::NewEnumerator0), B=1)
    Index 0 Param 2=50
    Index 0 Param 4=5
    Index 1 Param 1=4
    Index 1 Param 2=-0.07
    Index 1 Param 4=$Particles.Timeline.Time
    Index 2 Param 1=3
    Module.Index0Param0Write=true
    Module.Index0Param1Write=true
    Module.Index0Param2Write=true
    Module.Index0Param3Write=true
    Module.Index1Param0Write=true
    Module.Index1Param1Write=true
    Module.Index1Param2Write=true
    Module.Index1Param3Write=true
    Module.Index2Param0Write=true
    Module.Index2Param1Write=true
    Module.Index2Param2Write=true
    Module.Index2Param3Write=true
    Module.Index3Param0Write=true
    Module.Index3Param1Write=true
    Module.Index3Param2Write=true
    Module.Index3Param3Write=true
    Write Parameter Index 1=true
    Write Parameter Index 2=true
    Write Parameter Index 3=true
   )
   &ScaleMeshSize(Scale Factor=&VectorFromFloat(Value=1))
   /SetVariables(Particles.MeshOrientation=&MakeQuatFromAxisAngle(Axis=<1 0 0>, AngleInDegrees=90)) {isSetParameters=true, module=SetVariables_#4}
   &ScaleColor(
    Scale RGB=&VectorFromFloat(Value=&FNTimeOfDayScalar(Night Scale=1, Morning Scale=1, Evening Scale=1))
    ScaleA=true
    ScaleRGB=true
   ) {module=ScaleColor001}
   &ScaleMeshSize(Scale Factor=&VectorFromFloat(Value=&Lerp_Float(
      A=&Waveform(Global Input Position=-0.17, "Waveform [1]"=::NewEnumerator0, "[1] Frequency"=1, "[1] Amplitude Min/Max"=<1 1.2>)
      B=&Waveform(Global Input Position=-0.17, "Waveform [1]"=::NewEnumerator0, "[1] Phase Scale"=5, "[1] Frequency"=1, "[1] Amplitude Min/Max"=<1.8 2.2>)
      Alpha=&FloatFromCurve(
       FloatCurve=DataInterface() {propertyValues {Curve {keys [{interpMode=RCIM_Cubic, leaveTangent=1.8912, tangentMode=RCTM_Break, time=0, value=0}, <at 1 1>]}}}
       CurveIndex=$System.ReactiveValue
      )
     ))) {module=ScaleMeshSize001}
   &ScaleMeshSize(Scale Factor=&VectorFromFloat(Value=&Lerp_Float(A=0.8, B=1, Alpha=$System.PickaxeR))) {module=ScaleMeshSize002}
  ]
  rendererClasses [&NiagaraMeshRendererProperties]
  renderers [{rendererClass=&NiagaraMeshRendererProperties, rendererIndex=0}]
 }
 BurstRing001 {
  emitterUpdateScript [
   &EmitterState(
    Inactive Response=::NewEnumerator1
    Life Cycle Mode=::NewEnumerator1
    Loop Behavior=::NewEnumerator1
    Loop Duration Mode=::NewEnumerator1
    UseLoopDelay=false
   )
   &SpawnRate(SpawnRate=3, Use Spawn Probability=false) {module=SpawnRate001}
  ]
  particleSpawnScript [
   &InitializeParticle(
    Color=<rgba 0 3.3333 10 0.1>
    Lifetime=1.75
    Mesh Scale Mode=::NewEnumerator3
    Mesh Uniform Scale=0.07
    Position Offset=<0.6 -9 0.5>
    Sprite Size Mode=::NewEnumerator3
    UsePositionOffset=true
    Write Material Random=true
   )
   &Color(Color=&MultiplyLinearColor(
     Linear Color=$Particles.Color
     Scale Factor=&MakeVector4(
      X=1
      Y=1
      Z=1
      W=&MakeCustomFloatFromBool(
       Bool=&SetBoolByFloatComparison(A=&Subtract_Float(A=$Emitter.Age, B=$Emitter.float), B=-0.01)
       True Float=0.2
       False Float=1
      )
     )
    ))
   /SetVariables(Particles.bool:NiagaraBool=&SetBoolByFloatComparison(A=&Subtract_Float(A=$Emitter.Age, B=$Emitter.float), B=-0.01)) {isSetParameters=true, module=SetVariables_#3}
   &AddVelocity(Velocity=<0 -5 0>)
  ]
  particleUpdateScript [
   &ParticleState()
   &DynamicMaterialParameters(
    Index 0 Param 1=&MakeFloatFromVector(Vector=$Particles.Scale, Channel=::NewEnumerator0)
    Index 0 Param 2=&FloatFromCurve(FloatCurve=DataInterface() {propertyValues {Curve {keys [<at 0 1>,<at 1 2>]}}}, Scale Curve=3)
    Module.Index0Param0Write=true
    Module.Index0Param1Write=true
    Module.Index0Param2Write=true
    Module.Index0Param3Write=true
    Module.Index1Param0Write=true
    Module.Index1Param1Write=true
    Module.Index1Param2Write=true
    Module.Index1Param3Write=true
    Write Parameter Index 1=true
   )
   &ScaleMeshSize(Scale Factor=&VectorFromFloat(Value=&FloatFromCurve(FloatCurve=DataInterface() {propertyValues {Curve {keys [
          {interpMode=RCIM_Cubic, leaveTangent=3.5342, tangentMode=RCTM_Break, time=0, value=0.5}
          {arriveTangent=1.3693, interpMode=RCIM_Cubic, tangentMode=RCTM_Break, time=1, value=2}
         ]}}})))
   &ScaleColor(
    Scale Alpha=&MakeCustomFloatFromBool(
     Bool=$Particles.bool
     True Float=&FloatFromCurve(
      FloatCurve=DataInterface() {propertyValues {Curve {keys [
          {interpMode=RCIM_Cubic, leaveTangent=-3.4056, tangentMode=RCTM_Break, time=0, value=1}
          {arriveTangent=-0.0706, interpMode=RCIM_Cubic, tangentMode=RCTM_Break, time=1, value=0}
         ]}}}
      Scale Curve=2
     )
     False Float=&FloatFromCurve(
      FloatCurve=DataInterface() {propertyValues {Curve {keys [
          {interpMode=RCIM_Cubic, leaveTangent=-0.0192, tangentMode=RCTM_Break, time=0, value=1}
          {arriveTangent=-6.6328, interpMode=RCIM_Cubic, leaveTangent=-0.6773, tangentMode=RCTM_Break, time=0.2, value=0.2}
          {arriveTangent=-0.0456, interpMode=RCIM_Cubic, tangentMode=RCTM_Break, time=1, value=0}
         ]}}}
      Scale Curve=2
     )
    )
    Scale RGB=&VectorFromFloat(Value=&FloatFromCurve(FloatCurve=DataInterface() {propertyValues {Curve {keys [<at 0 1>, <at 0.1 3.5>, {arriveTangent=-2.6622, leaveTangent=-2.6622, time=0.2, value=2}, <at 1 0.45>]}}}))
    ScaleA=true
    ScaleRGB=true
   )
   /SetVariables(Particles.MeshOrientation=&MakeQuatFromAxisAngle(Axis=<1 0 0>, AngleInDegrees=90)) {isSetParameters=true, module=SetVariables_#4}
   &SolveForcesAndVelocity(Clamp Velocity=false, Limit Acceleration=false)
   &ScaleColor(
    Scale RGB=&VectorFromFloat(Value=&FNTimeOfDayScalar(Night Scale=1, Morning Scale=1, Evening Scale=1))
    ScaleA=true
    ScaleRGB=true
   ) {module=ScaleColor001}
   &ScaleColor(
    Scale Alpha=&FloatFromCurve(
     FloatCurve=DataInterface() {propertyValues {Curve {keys [
         {interpMode=RCIM_Cubic, leaveTangent=0.2728, tangentMode=RCTM_Break, time=0, value=1}
         {arriveTangent=0.9757, interpMode=RCIM_Cubic, leaveTangent=-0.1879, tangentMode=RCTM_Break, time=0.9, value=2}
         {arriveTangent=-4.7154, tangentMode=RCTM_User, time=1, value=1}
        ]}}}
     CurveIndex=$System.ReactiveValue
    )
    Scale RGB=&VectorFromFloat(Value=&Add_Float(
      A=&FloatFromCurve(
       FloatCurve=DataInterface() {propertyValues {Curve {keys [
           {interpMode=RCIM_Cubic, tangentMode=RCTM_Break, time=0, value=0}
           {arriveTangent=10.8803, interpMode=RCIM_Cubic, leaveTangent=-6.938, tangentMode=RCTM_Break, time=0.9, value=4.5}
           {arriveTangent=-82.1567, tangentMode=RCTM_User, time=1, value=0}
          ]}}}
       CurveIndex=$System.ReactiveValue
       Scale Curve=&FN_FrontEndFloatModifier(Gameplay=1, Front End=0.7)
      )
      B=1
     ))
    ScaleA=true
    ScaleRGB=true
   ) {module=ScaleColor002}
   &ScaleColor(Scale Alpha=$System.PickaxeR, ScaleA=true, ScaleRGB=true) {module=ScaleColor003}
  ]
  rendererClasses [&NiagaraMeshRendererProperties]
  renderers [{rendererClass=&NiagaraMeshRendererProperties, rendererIndex=0}]
 }
 BurstRing002 {
  emitterSpawnScript [/SetVariables(Emitter.float=-100) {isSetParameters=true, module=SetVariables_#5}]
  emitterUpdateScript [
   &EmitterState(
    Inactive Response=::NewEnumerator1
    Life Cycle Mode=::NewEnumerator1
    Loop Behavior=::NewEnumerator1
    Loop Duration Mode=::NewEnumerator1
    UseLoopDelay=false
   )
   &SpawnPerFrame(Spawn:NiagaraBool=$System.DoOnce.Execute, Use Spawn Probability=false)
   /SetVariables(Emitter.float:NiagaraFloat=&MakeCustomFloatFromBool(
     Bool=$System.DoOnce.Execute
     True Float=&self:ScratchDynamicInput_01(float:NiagaraFloat=$Emitter.Age, float001:NiagaraFloat=0.033)
     False Float=$Emitter.float
    )) {enabled=false, isSetParameters=true, module=SetVariables_#6}
   &SpawnBurst_Instantaneous(Spawn Group=1, Spawn Time=$Emitter.float, Use Spawn Probability=false, UseLoopCountLimit=false) {enabled=false}
  ]
  particleSpawnScript [&InitializeParticle(
    Color=<rgba 1.625 1.5 3 0.2>
    Lifetime=0.8
    Mesh Scale Mode=::NewEnumerator3
    Mesh Uniform Scale=0.08
    Position Offset=<0.6 -11 0.5>
    Sprite Size Mode=::NewEnumerator3
    Uniform Sprite Size=32.3145
    UsePositionOffset=true
    Write Material Random=true
   )]
  particleUpdateScript [
   &ParticleState()
   &DynamicMaterialParameters(
    Index 0 Param 1=&MakeFloatFromVector(Vector=$Particles.Scale, Channel=::NewEnumerator0)
    Index 0 Param 2=&FloatFromCurve(FloatCurve=DataInterface() {propertyValues {Curve {keys [<at 0 1>,<at 1 2>]}}}, Scale Curve=5)
    Module.Index0Param0Write=true
    Module.Index0Param1Write=true
    Module.Index0Param2Write=true
    Module.Index0Param3Write=true
    Module.Index1Param0Write=true
    Module.Index1Param1Write=true
    Module.Index1Param2Write=true
    Module.Index1Param3Write=true
    Write Parameter Index 1=true
   )
   &ScaleMeshSize(Scale Factor=&VectorFromFloat(Value=&FloatFromCurve(FloatCurve=DataInterface() {propertyValues {Curve {keys [
          {interpMode=RCIM_Cubic, leaveTangent=3.5342, tangentMode=RCTM_Break, time=0, value=0.5}
          {arriveTangent=1.3693, interpMode=RCIM_Cubic, tangentMode=RCTM_Break, time=1, value=2}
         ]}}})))
   &ScaleColor(
    Scale Alpha=&MakeCustomFloatFromBool(
     Bool=$Particles.bool
     True Float=&FloatFromCurve(
      FloatCurve=DataInterface() {propertyValues {Curve {keys [
          {interpMode=RCIM_Cubic, leaveTangent=-3.4056, tangentMode=RCTM_Break, time=0, value=1}
          {arriveTangent=-0.3569, interpMode=RCIM_Cubic, leaveTangent=-0.3569, tangentMode=RCTM_User, time=0.3, value=0.1}
          {arriveTangent=-0.0706, interpMode=RCIM_Cubic, tangentMode=RCTM_Break, time=1, value=0}
         ]}}}
      Scale Curve=2
     )
     False Float=&FloatFromCurve(
      FloatCurve=DataInterface() {propertyValues {Curve {keys [
          {interpMode=RCIM_Cubic, leaveTangent=-0.0192, tangentMode=RCTM_Break, time=0, value=1}
          {arriveTangent=-6.6328, interpMode=RCIM_Cubic, leaveTangent=-0.6773, tangentMode=RCTM_Break, time=0.2, value=0.2}
          {arriveTangent=-0.0456, interpMode=RCIM_Cubic, tangentMode=RCTM_Break, time=1, value=0}
         ]}}}
      Scale Curve=2
     )
    )
    Scale RGB=&VectorFromFloat(Value=&FloatFromCurve(FloatCurve=DataInterface() {propertyValues {Curve {keys [<at 0 1>, <at 0.1 3.5>, {arriveTangent=-2.6622, leaveTangent=-2.6622, time=0.2, value=2}, <at 1 0.45>]}}}))
    ScaleA=true
    ScaleRGB=true
   )
   /SetVariables(Particles.MeshOrientation=&MultiplyQuaternion(Quaternion A=<0 0 0 1>, Quaternion B=&MakeQuatFromAxisAngle(Axis=<1 0 0>, AngleInDegrees=90))) {isSetParameters=true, module=SetVariables_#4}
   &ScaleColor(
    Scale RGB=&VectorFromFloat(Value=&FNTimeOfDayScalar(Night Scale=1, Morning Scale=1, Evening Scale=1))
    ScaleA=true
    ScaleRGB=true
   ) {module=ScaleColor001}
  ]
  rendererClasses [&NiagaraMeshRendererProperties, &NiagaraSpriteRendererProperties]
  renderers [
   {rendererClass=&NiagaraMeshRendererProperties, rendererIndex=0}
   {rendererClass=&NiagaraSpriteRendererProperties, rendererIndex=1}
  ]
 }
 BurstRing003 {
  emitterSpawnScript [/SetVariables(Emitter.float=-100) {isSetParameters=true, module=SetVariables_#5}]
  emitterUpdateScript [
   &EmitterState(
    Inactive Response=::NewEnumerator1
    Life Cycle Mode=::NewEnumerator1
    Loop Behavior=::NewEnumerator1
    Loop Duration Mode=::NewEnumerator1
    UseLoopDelay=false
   )
   &SpawnPerFrame(Spawn:NiagaraBool=$System.DoOnce.Execute, Use Spawn Probability=false)
   /SetVariables(Emitter.float:NiagaraFloat=&MakeCustomFloatFromBool(
     Bool=$System.DoOnce.Execute
     True Float=&self:ScratchDynamicInput_01(float:NiagaraFloat=$Emitter.Age, float001:NiagaraFloat=0.033)
     False Float=$Emitter.float
    )) {enabled=false, isSetParameters=true, module=SetVariables_#6}
   &SpawnBurst_Instantaneous(Spawn Group=1, Spawn Time=$Emitter.float, Use Spawn Probability=false, UseLoopCountLimit=false) {enabled=false}
  ]
  particleSpawnScript [&InitializeParticle(
    Color=<rgba 0.1 0.1 1 1>
    Lifetime=0.9
    Mesh Scale Mode=::NewEnumerator3
    Mesh Uniform Scale=&MakeCustomFloatFromBool(
     Bool=&SetBoolByFloatComparison(A=&Subtract_Float(A=$Emitter.Age, B=$Emitter.float), B=-0.01)
     True Float=0.08
     False Float=0.08
    )
    Position Offset=<0.6 -10 0.5>
    Sprite Size Mode=::NewEnumerator3
    Uniform Sprite Size=32.3145
    UsePositionOffset=true
    Write Material Random=true
   )]
  particleUpdateScript [
   &ParticleState()
   &ScaleColor(
    Scale Alpha=&MakeCustomFloatFromBool(
     Bool=$Particles.bool
     True Float=&FloatFromCurve(
      FloatCurve=DataInterface() {propertyValues {Curve {keys [
          {interpMode=RCIM_Cubic, leaveTangent=-3.4056, tangentMode=RCTM_Break, time=0, value=1}
          {arriveTangent=-0.3569, interpMode=RCIM_Cubic, leaveTangent=-0.3569, tangentMode=RCTM_User, time=0.3, value=0.1}
          {arriveTangent=-0.0706, interpMode=RCIM_Cubic, tangentMode=RCTM_Break, time=1, value=0}
         ]}}}
      Scale Curve=2
     )
     False Float=&FloatFromCurve(
      FloatCurve=DataInterface() {propertyValues {Curve {keys [
          {interpMode=RCIM_Cubic, leaveTangent=-0.0192, tangentMode=RCTM_Break, time=0, value=1}
          {arriveTangent=-6.6328, interpMode=RCIM_Cubic, leaveTangent=-0.6773, tangentMode=RCTM_Break, time=0.2, value=0.2}
          {arriveTangent=-0.0456, interpMode=RCIM_Cubic, tangentMode=RCTM_Break, time=1, value=0}
         ]}}}
      Scale Curve=2
     )
    )
    Scale RGB=&VectorFromFloat(Value=&FloatFromCurve(FloatCurve=DataInterface() {propertyValues {Curve {keys [<at 0 1>, <at 0.1 3.5>, {arriveTangent=-2.6622, leaveTangent=-2.6622, time=0.2, value=2}, <at 1 0.45>]}}}))
    ScaleA=true
    ScaleRGB=true
   )
   &ScaleColor(
    Scale RGB=&Multiply_VectorByFloat(
     Vector=&VectorFromFloat(Value=&FNTimeOfDayScalar(Night Scale=1, Morning Scale=1, Evening Scale=1))
     Float=&FN_FrontEndFloatModifier(Gameplay=1, Front End=2)
    )
    ScaleA=true
    ScaleRGB=true
   ) {module=ScaleColor001}
  ]
  rendererClasses [&NiagaraSpriteRendererProperties]
  renderers [{rendererClass=&NiagaraSpriteRendererProperties}]
 }
 BurstRing004 {
  emitterUpdateScript [
   &EmitterState(
    Inactive Response=::NewEnumerator1
    Life Cycle Mode=::NewEnumerator1
    Loop Behavior=::NewEnumerator1
    Loop Duration Mode=::NewEnumerator1
    UseLoopDelay=false
   )
   &SpawnBurst_Instantaneous(Use Spawn Probability=false, UseLoopCountLimit=false)
  ]
  particleSpawnScript [
   &InitializeParticle(
    Color=<rgba 0.4167 0 5 1>
    Lifetime=&Multiply_Float(A=1, B=&FN_FrontEndFloatModifier(Gameplay=1, Front End=0.7))
    Mesh Scale Mode=::NewEnumerator3
    Mesh Uniform Scale=0.055
    Position Offset=<0.6 -9.87 0.5>
    Sprite Size Mode=::NewEnumerator3
    UsePositionOffset=true
    Write Material Random=true
   )
   &Color(Color=&MultiplyLinearColor(Linear Color=$Particles.Color, Scale Factor=&MakeVector4(X=1, Y=1, Z=1, W=0.15)))
   /SetVariables(Particles.bool:NiagaraBool=&SetBoolByFloatComparison(A=&Subtract_Float(A=$Emitter.Age, B=$Emitter.float), B=-0.01)) {isSetParameters=true, module=SetVariables_#3}
  ]
  particleUpdateScript [
   &ParticleState(Kill Particles When Lifetime Has Elapsed=false, Let Infinitely Lived Particles Die When Emitter Deactivates=true)
   &Timeline(
    Play Rate=&FloatFromCurve(
     FloatCurve=DataInterface() {propertyValues {Curve {keys [<at 0 1>,<at 0.9 2>,<at 1 1>]}}}
     CurveIndex=$System.ReactiveValue
     Scale Curve=&Lerp_Float(A=0.01, B=0.015, Alpha=$System.PickaxeR)
    )
    Timeline End=100000
   )
   &CameraOffset()
   &DynamicMaterialParameters(
    Index 0 Param 1=&Multiply_Float(A=&MakeFloatFromVector(Vector=$Particles.Scale, Channel=::NewEnumerator0), B=1)
    Index 0 Param 2=50
    Index 0 Param 4=5
    Index 1 Param 1=15
    Index 1 Param 2=-0.07
    Index 1 Param 3=0.05
    Index 1 Param 4=$Particles.Timeline.Time
    Index 2 Param 1=3
    Module.Index0Param0Write=true
    Module.Index0Param1Write=true
    Module.Index0Param2Write=true
    Module.Index0Param3Write=true
    Module.Index1Param0Write=true
    Module.Index1Param1Write=true
    Module.Index1Param2Write=true
    Module.Index1Param3Write=true
    Module.Index2Param0Write=true
    Module.Index2Param1Write=true
    Module.Index2Param2Write=true
    Module.Index2Param3Write=true
    Module.Index3Param0Write=true
    Module.Index3Param1Write=true
    Module.Index3Param2Write=true
    Module.Index3Param3Write=true
    Write Parameter Index 1=true
    Write Parameter Index 2=true
    Write Parameter Index 3=true
   )
   &ScaleMeshSize(Scale Factor=&VectorFromFloat(Value=1))
   /SetVariables(Particles.MeshOrientation=&MakeQuatFromAxisAngle(Axis=<1 0 0>, AngleInDegrees=90)) {isSetParameters=true, module=SetVariables_#4}
   &ScaleColor(
    Scale RGB=&VectorFromFloat(Value=&FNTimeOfDayScalar(Night Scale=1, Morning Scale=1, Evening Scale=1))
    ScaleA=true
    ScaleRGB=true
   ) {module=ScaleColor001}
   &ScaleMeshSize(Scale Factor=&VectorFromFloat(Value=&Lerp_Float(
      A=&Waveform(Global Input Position=-0.17, "Waveform [1]"=::NewEnumerator0, "[1] Frequency"=1, "[1] Amplitude Min/Max"=<1 1.2>)
      B=&Waveform(Global Input Position=-0.17, "Waveform [1]"=::NewEnumerator0, "[1] Phase Scale"=5, "[1] Frequency"=1, "[1] Amplitude Min/Max"=<1.8 2.2>)
      Alpha=&FloatFromCurve(
       FloatCurve=DataInterface() {propertyValues {Curve {keys [{interpMode=RCIM_Cubic, leaveTangent=1.8912, tangentMode=RCTM_Break, time=0, value=0}, <at 1 1>]}}}
       CurveIndex=$System.ReactiveValue
      )
     ))) {module=ScaleMeshSize001}
   &ScaleMeshSize(Scale Factor=&VectorFromFloat(Value=&Lerp_Float(A=0.8, B=1.5, Alpha=$System.PickaxeR))) {module=ScaleMeshSize002}
  ]
  rendererClasses [&NiagaraMeshRendererProperties]
  renderers [{rendererClass=&NiagaraMeshRendererProperties, rendererIndex=0}]
 }
 Glow {
  emitterUpdateScript [
   &EmitterState(
    Inactive Response=::NewEnumerator1
    Life Cycle Mode=::NewEnumerator1
    Loop Behavior=::NewEnumerator1
    Loop Duration Mode=::NewEnumerator1
    UseLoopDelay=false
   )
   &SpawnBurst_Instantaneous(Use Spawn Probability=false, UseLoopCountLimit=false)
  ]
  particleSpawnScript [&InitializeParticle(
    Color=<rgba 0.2 0.2 0.2 0.5>
    Lifetime=5
    Mesh Scale Mode=::NewEnumerator3
    Mesh Uniform Scale=0.1
    Position Offset=<0.6 -10 0.5>
    Sprite Size Mode=::NewEnumerator3
    Uniform Sprite Size=15
    UsePositionOffset=true
    Write Material Random=true
   )]
  particleUpdateScript [
   &ParticleState(Kill Particles When Lifetime Has Elapsed=false, Let Infinitely Lived Particles Die When Emitter Deactivates=true)
   &CameraOffset(Camera Offset Amount=1)
   &SpriteFacingAndAlignment(Sprite Alignment=<0 0 1>, Sprite Facing=<0 1 0>, WriteAlignment=true, WriteFacing=true)
  ]
  rendererClasses [&NiagaraSpriteRendererProperties]
  renderers [{rendererClass=&NiagaraSpriteRendererProperties}]
 }
 Glow001 {
  emitterUpdateScript [
   &EmitterState(
    Inactive Response=::NewEnumerator1
    Life Cycle Mode=::NewEnumerator1
    Loop Behavior=::NewEnumerator1
    Loop Duration Mode=::NewEnumerator1
    UseLoopDelay=false
   )
   &SpawnBurst_Instantaneous(Use Spawn Probability=false, UseLoopCountLimit=false)
  ]
  particleSpawnScript [&InitializeParticle(
    Color=<rgba 0 0.2083 0.5 1>
    Lifetime=5
    Mesh Scale Mode=::NewEnumerator3
    Position Offset=<0 -10 0>
    Sprite Size Mode=::NewEnumerator3
    Uniform Sprite Size=30
    UsePositionOffset=true
    Write Material Random=true
   )]
  particleUpdateScript [
   &ParticleState(Kill Particles When Lifetime Has Elapsed=false, Let Infinitely Lived Particles Die When Emitter Deactivates=true)
   &CameraOffset(Camera Offset Amount=20)
   /SetVariables(
    Particles.Color2=&HueShiftLinearColor(Hue=-0.2, Saturation=1, Value=0.3, Linear Color=$Particles.Color)
    Particles.SpriteSize2=&Multiply_Vector2D_ByFloat(Float=2, Vector2D=$Particles.SpriteSize)
   ) {isSetParameters=true}
   &self:-----------Reactive----------() {module=-----------Reactive----------}
   &ScaleColor(
    Scale Alpha=&FloatFromCurve(
     FloatCurve=DataInterface() {propertyValues {Curve {keys [
         {interpMode=RCIM_Cubic, leaveTangent=0.2728, tangentMode=RCTM_Break, time=0, value=1}
         {arriveTangent=0.9757, interpMode=RCIM_Cubic, leaveTangent=-0.1879, tangentMode=RCTM_Break, time=0.9, value=1.4}
         {arriveTangent=-4.7154, tangentMode=RCTM_User, time=1, value=1}
        ]}}}
     CurveIndex=$System.ReactiveValue
    )
    Scale RGB=&VectorFromFloat(Value=&Add_Float(
      A=&FloatFromCurve(
       FloatCurve=DataInterface() {propertyValues {Curve {keys [
           {interpMode=RCIM_Cubic, tangentMode=RCTM_Break, time=0, value=0}
           {arriveTangent=10.8803, interpMode=RCIM_Cubic, leaveTangent=-6.938, tangentMode=RCTM_Break, time=0.9, value=3.5}
           {arriveTangent=-82.1567, tangentMode=RCTM_User, time=1, value=0}
          ]}}}
       CurveIndex=$System.ReactiveValue
       Scale Curve=&FN_FrontEndFloatModifier(Gameplay=1, Front End=0.7)
      )
      B=1
     ))
    ScaleA=true
    ScaleRGB=true
   ) {module=ScaleColor001}
   &ScaleSpriteSize(
    Uniform Curve Index=$System.ReactiveValue
    Uniform Curve Sprite Scale=DataInterface() {propertyValues {Curve {keys [{tangentMode=RCTM_User, time=0, value=1}, {tangentMode=RCTM_User, time=0.9, value=1.2}, <at 1 1>]}}}
   )
   &ScaleColor(
    Scale RGB=&VectorFromFloat(Value=&FNTimeOfDayScalar(Night Scale=1, Morning Scale=1, Evening Scale=1))
    ScaleA=true
    ScaleRGB=true
   )
   &FN_Light_Attributes(
    Color=<rgba 0.15 0.2917 1 1>
    Create Color Attribute=true
    Create Position Attribute=true
    Create Radius Attribute=true
    Position=&AddVectorToPosition(Position=$Particles.Position, Vector=<0 -6 0>)
    Radius=30
    Scale Color=&Multiply_Float(
     A=&Add_Float(
      A=&Waveform(
       "Waveform [1]"=::NewEnumerator0
       "[1] Frequency"=&Lerp_Float(A=1, B=3, Alpha=$NPC.FortniteNPC.EventAudio_Amplitude)
       "[1] Amplitude Min/Max"=<0.3 0.4>
      )
      B=&Divide_Float(A=$NPC.FortniteNPC.EventAudio_Amplitude, B=3)
     )
     B=5000
    )
   )
  ]
  rendererClasses [
   &NiagaraSpriteRendererProperties
   &NiagaraLightRendererProperties
   &NiagaraMeshRendererProperties
  ]
  renderers [
   {rendererClass=&NiagaraSpriteRendererProperties}
   {rendererClass=&NiagaraSpriteRendererProperties, rendererIndex=1}
   {rendererClass=&NiagaraLightRendererProperties, rendererIndex=2}
   {rendererClass=&NiagaraMeshRendererProperties, rendererIndex=3}
  ]
 }
 Ring {
  emitterUpdateScript [
   &EmitterState(
    Inactive Response=::NewEnumerator1
    Life Cycle Mode=::NewEnumerator1
    Loop Behavior=::NewEnumerator1
    Loop Duration Mode=::NewEnumerator1
    UseLoopDelay=false
   )
   &SpawnBurst_Instantaneous(Spawn Count=5, Use Spawn Probability=false, UseLoopCountLimit=false)
  ]
  particleSpawnScript [
   &InitializeParticle(
    Color=&SelectLinearColorFromArray(
     Array Sampling Mode=::NewEnumerator1
     Direct Array Index=&ModuloInt(A=&ReturnParticleIDAttributes(), B=5)
     Color Selection Array=DataInterface() {propertyValues {ColorData [<rgba 0 0.091 0.2 2>, <rgba 0 0.6667 1 1>, <rgba 0.2 0.56 2 1>, <rgba 0.225 0.48 1.5 0.3>, <rgba 0 1 1.5 0.3>]}}
     Random Seed=?
    )
    Lifetime=5
    Mesh Scale Mode=::NewEnumerator3
    Mesh Uniform Scale=&SelectFloatFromArray(
     Direct Array Index=&ModuloInt(A=&ReturnParticleIDAttributes(), B=5)
     Float Selection Array=DataInterface() {propertyValues {FloatData [0.03,0.018,0.03,0.03,0.018]}}
    )
    Position Offset=&SelectVectorFromArray(
     Direct Array Index=&ModuloInt(A=&ReturnParticleIDAttributes(), B=5)
     Vector Selection Array=DataInterface() {propertyValues {FloatData [<0.6 -11 0.5>, <0.6 -9.5 0.5>, <0.6 -11 0.5>, <0.6 -11 0.5>, <0.6 -9.5 0.5>]}}
    )
    Sprite Size Mode=::NewEnumerator3
    UsePositionOffset=false
    Write Material Random=true
   )
   /SetVariables(Particles.VisibilityTag=&SelectIntFromArray(
     Direct Array Index=&ModuloInt(A=&ReturnParticleIDAttributes(), B=5)
     Int Selection Array=DataInterface() {propertyValues {IntData [0,1,1,2,2]}}
    )) {isSetParameters=true}
   &KillParticles(Kill Particles=&SetBoolByFloatComparison(
     A=&FN_PickFloatDependOnPlatform(
      TargetPlatformActive=-1
      TargetPlatformNoActive=&SelectFloatFromArray(
       Direct Array Index=&ModuloInt(A=&ReturnParticleIDAttributes(), B=5)
       Float Selection Array=DataInterface() {propertyValues {FloatData [-1,-1,-1,-1,1]}}
      )
     )
     B=0
    ))
   &InitialMeshOrientation(Mesh Axis To Orient=<0 1 0>, Mesh Orientation Mode=::NewEnumerator2, Orientation Vector=<0 0 1>, Use Rotation Vector=false)
  ]
  particleUpdateScript [
   &ParticleState(Kill Particles When Lifetime Has Elapsed=false, Let Infinitely Lived Particles Die When Emitter Deactivates=true)
   &ScaleMeshSize(Scale Factor=&VectorFromFloat(Value=&Lerp_Float(A=1.7, B=2, Alpha=$System.PickaxeR))) {module=ScaleMeshSize001}
   &Timeline(
    Initial Start Time=256.258
    Play Rate=&FloatFromCurve(
     FloatCurve=DataInterface() {propertyValues {Curve {keys [<at 0 1>,<at 0.9 2>,<at 1 1>]}}}
     CurveIndex=$System.ReactiveValue
     Scale Curve=&Lerp_Float(A=0.8, B=0.9, Alpha=$System.PickaxeR)
    )
    Timeline Beginning=256.258
    Timeline End=100000
   )
   &DynamicMaterialParameters(
    Index 0 Param 1=&MakeFloatFromVector(Vector=$Particles.Scale, Channel=::NewEnumerator0)
    Index 0 Param 2=5
    Index 0 Param 3=&Lerp_Float(
     A=0
     B=&Waveform(Global Input Position=-0.17, "Waveform [1]"=::NewEnumerator0, "[1] Phase Scale"=3, "[1] Frequency"=1, "[1] Amplitude Min/Max"=<0 0>)
     Alpha=&Power(a=$System.ReactiveValue, b=3)
    )
    Index 0 Param 4=&Waveform(Global Input Position=128, "Waveform [1]"=::NewEnumerator12, "[1] Phase Scale"=2, "[1] Frequency"=1, "[1] Amplitude Min/Max"=<-0.2 0.2>)
    Index 1 Param 1=3
    Index 1 Param 2=&Multiply_Float(
     A=&SelectFloatFromArray(
      Direct Array Index=&ModuloInt(A=&ReturnParticleIDAttributes(), B=5)
      Float Selection Array=DataInterface() {propertyValues {FloatData [1,1,1,-0.75,0.2]}}
     )
     B=0.7
    )
    Index 1 Param 3=&Lerp_Float(
     A=&FloatFromCurve(
      FloatCurve=DataInterface() {propertyValues {Curve {keys [{interpMode=RCIM_Cubic, leaveTangent=0.7531, tangentMode=RCTM_Break, time=0, value=-0.5}, <at 1 0.4>]}}}
      CurveIndex=$System.ReactiveValue
     )
     B=0.5
     Alpha=$System.PickaxeR
    )
    Index 1 Param 4=$Particles.Timeline.Time
    Index 2 Param 1=&FloatFromCurve(
     FloatCurve=DataInterface() {propertyValues {Curve {keys [{interpMode=RCIM_Cubic, tangentMode=RCTM_Break, time=0, value=1}, <at 1 3.5>]}}}
     CurveIndex=$System.ReactiveValue
    )
    Module.Index0Param0Write=true
    Module.Index0Param1Write=true
    Module.Index0Param2Write=true
    Module.Index0Param3Write=true
    Module.Index1Param0Write=true
    Module.Index1Param1Write=true
    Module.Index1Param2Write=true
    Module.Index1Param3Write=true
    Module.Index2Param0Write=true
    Module.Index2Param1Write=true
    Module.Index2Param2Write=true
    Module.Index2Param3Write=true
    Write Parameter Index 1=true
    Write Parameter Index 2=true
   )
   &ScaleMeshSize(Scale Factor=&VectorFromFloat(Value=&Lerp_Float(
      A=&Waveform(Global Input Position=-0.17, "Waveform [1]"=::NewEnumerator0, "[1] Frequency"=1, "[1] Amplitude Min/Max"=<1 1.2>)
      B=&Multiply_Float(
       A=&Waveform(Global Input Position=-0.17, "Waveform [1]"=::NewEnumerator0, "[1] Phase Scale"=5, "[1] Frequency"=1, "[1] Amplitude Min/Max"=<1.3 1.5>)
       B=&SelectFloatFromArray(
        Direct Array Index=&ModuloInt(A=&ReturnParticleIDAttributes(), B=5)
        Direct Array Mode=::NewEnumerator1
        Float Selection Array=DataInterface() {propertyValues {FloatData [1,1,1,1,1]}}
       )
      )
      Alpha=&FloatFromCurve(
       FloatCurve=DataInterface() {propertyValues {Curve {keys [{interpMode=RCIM_Cubic, leaveTangent=1.8912, tangentMode=RCTM_Break, time=0, value=0}, <at 1 1>]}}}
       CurveIndex=$System.ReactiveValue
      )
     )))
   &ScaleColor(
    Scale RGB=&VectorFromFloat(Value=&Add_Float(
      A=&Multiply_Float(
       A=&FloatFromCurve(
        FloatCurve=DataInterface() {propertyValues {Curve {keys [
            {interpMode=RCIM_Cubic, tangentMode=RCTM_Break, time=0, value=0}
            {arriveTangent=3.7964, interpMode=RCIM_Cubic, leaveTangent=1.1533, tangentMode=RCTM_Break, time=0.9, value=2}
            <at 1 0>
           ]}}}
        CurveIndex=$System.ReactiveValue
        Scale Curve=&SelectFloatFromArray(
         Direct Array Index=&ModuloInt(A=&ReturnParticleIDAttributes(), B=5)
         Float Selection Array=DataInterface() {propertyValues {FloatData [0,1.6,0.75,2.5,2.5]}}
        )
       )
       B=&FN_FrontEndFloatModifier(Gameplay=1, Front End=0.5)
      )
      B=&Lerp_Float(
       A=&SelectFloatFromArray(
        Direct Array Index=&ModuloInt(A=&ReturnParticleIDAttributes(), B=5)
        Float Selection Array=DataInterface() {propertyValues {FloatData [1,1,1,0.3,0.8]}}
       )
       B=1
       Alpha=$User.PickaxeReactive
      )
     ))
    ScaleA=true
    ScaleRGB=true
   ) {module=ScaleColor001}
   &ScaleColor(
    Scale RGB=&VectorFromFloat(Value=&FN_FrontEndFloatModifier(Gameplay=&FNTimeOfDayScalar(Night Scale=1, Morning Scale=1, Evening Scale=1), Front End=1.2))
    ScaleA=true
    ScaleRGB=true
   )
   &SystemLocation(
    MaskSpawnGroup=false
    Offset=&Multiply_Vector(
     A=&SelectVectorFromArray(
      Direct Array Index=&ModuloInt(A=&ReturnParticleIDAttributes(), B=5)
      Vector Selection Array=DataInterface() {propertyValues {FloatData [<0.6 -11 0.5>, <0.6 -9.5 0.5>, <0.6 -11 0.5>, <0.6 -11 0.5>, <0.6 -9.5 0.5>]}}
     )
     B=&MakeVector(X=1, Y=&Lerp_Float(A=0.88, B=1, Alpha=$System.PickaxeR), Z=1)
    )
   )
  ]
  rendererClasses [&NiagaraMeshRendererProperties]
  renderers [
   {rendererClass=&NiagaraMeshRendererProperties, rendererIndex=0}
   {rendererClass=&NiagaraMeshRendererProperties, rendererIndex=1}
   {rendererClass=&NiagaraMeshRendererProperties, rendererIndex=2}
  ]
 }
 SoundCube {
  emitterUpdateScript [
   &EmitterState(
    Inactive Response=::NewEnumerator1
    Life Cycle Mode=::NewEnumerator1
    Loop Behavior=::NewEnumerator1
    Loop Duration Mode=::NewEnumerator1
    UseLoopDelay=false
   )
   &SpawnBurst_Instantaneous(Spawn Count=9, Use Spawn Probability=false, UseLoopCountLimit=false)
  ]
  particleSpawnScript [&InitializeParticle(
    Color Mode=::NewEnumerator0
    Lifetime=5
    Mesh Scale Mode=::NewEnumerator3
    Mesh Uniform Scale=0.6
    Position Offset=<0.6 -11 0.5>
    UsePositionOffset=true
    Write Material Random=true
   )]
  particleUpdateScript [
   &ParticleState(Kill Particles When Lifetime Has Elapsed=false, Let Infinitely Lived Particles Die When Emitter Deactivates=true)
   &ScaleMeshSize(Scale Factor=&MakeVector(
     X=1
     Y=1
     Z=&Waveform(
      Global Input Position=&RandomRangeFloat(Minimum=0, Maximum=15)
      "Waveform [1]"=::NewEnumerator12
      "[1] Phase Scale"=10
      "[1] Frequency"=1
      "[1] Amplitude Min/Max"=&MakeVector2D(X=0, Y=&Lerp_Float(A=8, B=15, Alpha=$System.ReactiveValue))
     )
    ))
   /SetVariables(Particles.MeshOrientation=&MakeQuatFromAxisAngle(Axis=<1 0 0>, AngleInDegrees=90)) {isSetParameters=true}
   /SetVariables(Particles.PositionB:NiagaraPosition=&AddVectorToPosition(
     Position=$Particles.Position
     Vector=&Multiply_VectorByFloat(
      Vector=&RotateVectorByQuaternion(
       Vector To Rotate=&SelectVectorFromArray(
        Direct Array Index=&ModuloInt(A=&ReturnParticleIDAttributes(), B=9)
        Direct Array Mode=::NewEnumerator1
        Vector Selection Array=DataInterface() {propertyValues {FloatData [<1 1 0>, <1 0 0>, <1 -1 0>, <0 1 0>, <0 0 0>, <0 -1 0>, <-1 1 0>, <-1 0 0>, <-1 -1 0>]}}
       )
       Quaternion=$Particles.MeshOrientation
      )
      Float=1.3
     )
    )) {isSetParameters=true}
   &ScaleColor(
    Scale Alpha=&Waveform(
     Global Input Position=&RandomRangeFloat(Minimum=0, Maximum=15)
     "Waveform [1]"=::NewEnumerator12
     "[1] Phase Scale"=10
     "[1] Frequency"=1
     "[1] Amplitude Min/Max"=<0 0.5>
    )
    ScaleA=true
    ScaleRGB=true
   )
   &ScaleColor(
    Scale Alpha=&FloatFromCurve(
     FloatCurve=DataInterface() {propertyValues {Curve {keys [
         {interpMode=RCIM_Cubic, leaveTangent=0.431, tangentMode=RCTM_Break, time=0, value=1}
         {arriveTangent=3.4786, interpMode=RCIM_Cubic, leaveTangent=-19.0618, tangentMode=RCTM_Break, time=0.9, value=2}
         <at 1 1>
        ]}}}
     CurveIndex=$System.ReactiveValue
    )
    Scale RGB=&VectorFromFloat(Value=&FloatFromCurve(
      FloatCurve=DataInterface() {propertyValues {Curve {keys [
          {interpMode=RCIM_Cubic, leaveTangent=3.4065, tangentMode=RCTM_Break, time=0, value=1}
          {arriveTangent=0.8198, interpMode=RCIM_Cubic, leaveTangent=-19.0618, tangentMode=RCTM_Break, time=0.9, value=4}
          <at 1 1>
         ]}}}
      CurveIndex=$System.ReactiveValue
     ))
    ScaleA=true
    ScaleRGB=true
   ) {module=ScaleColor001}
   &ScaleColor(
    Scale RGB=&VectorFromFloat(Value=&Lerp_Float(A=1, B=2, Alpha=$System.PickaxeR))
    ScaleA=true
    ScaleRGB=true
   ) {module=ScaleColor002}
  ]
  rendererClasses [&NiagaraSpriteRendererProperties, &NiagaraMeshRendererProperties]
  renderers [
   {rendererClass=&NiagaraSpriteRendererProperties}
   {rendererClass=&NiagaraMeshRendererProperties, rendererIndex=1}
  ]
 }
}
summary {
 systemName=NS_TechnoAwakenBackpack
 emitters [
  {emitterName=Ring, rendererClasses [&NiagaraMeshRendererProperties]}
  {
   emitterName=Glow001
   rendererClasses [
    &NiagaraSpriteRendererProperties
    &NiagaraLightRendererProperties
    &NiagaraMeshRendererProperties
   ]
  }
  {emitterName=SoundCube, rendererClasses [&NiagaraSpriteRendererProperties, &NiagaraMeshRendererProperties]}
  {emitterName=Glow, rendererClasses [&NiagaraSpriteRendererProperties]}
  {emitterName=BurstRing, rendererClasses [&NiagaraMeshRendererProperties]}
  {emitterName=BurstRing001, rendererClasses [&NiagaraMeshRendererProperties]}
  {emitterName=BurstRing002, rendererClasses [&NiagaraMeshRendererProperties, &NiagaraSpriteRendererProperties]}
  {emitterName=BurstRing003, rendererClasses [&NiagaraSpriteRendererProperties]}
  {emitterName=BurstRing004, rendererClasses [&NiagaraMeshRendererProperties]}
 ]
 userVariables [
  User.bIsEmote {type=NiagaraBool}
  User.bIsFrontEnd {type=NiagaraBool}
  User.bIsFrontEndPreview {type=NiagaraBool}
  User.Kill {type=NiagaraFloat, defaultValue=0, description=""}
  User.PickaxeReactive {type=NiagaraFloat, defaultValue=1, description=""}
 ]
}
