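# Physics asset for the TechnoAwaken backpack cosmetic, giving collision bodies and joint limits to its dynamic arm and left and right wire bones.
# A box covers the pack, capsules cover the arm and wire bones, and the left wire master adds a box.
# The arm joint locks both swings and has a twist limit, while wire joints have tight twist and uneven swing limits.
# Both left wire segments constrain directly to their master rather than chaining, and no constraint links to the pack.
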PhysicsAsset /BRCosmetics/Accessories/FORT_Backpacks/Backpack_M_MED_TechnoAwaken/Meshes/M_MED_TechnoAwaken_Pack_Physics
bodies {
 dyn_arm_01 [{
   shapeType=Capsule
   length=9.2022
   radius=0.7414
   shapeName=dyn_arm_01_capsule
   center=<1.3136 2.8705 -5.7561>
   rotation=<rot -25.8705 -90.0001 -0.118>
  }]
 dyn_arm_master [{shapeType=Capsule, length=2, radius=1, shapeName=dyn_arm_master_capsule, rotation=<rot -90 180 180>}]
 dyn_wire_l_01 [{
   shapeType=Capsule
   length=4.4358
   radius=0.5698
   shapeName=dyn_wire_l_01_capsule
   center=<-4.6306 -0.013 0.7313>
   rotation=<rot 25.3124 0 -92.8128>
  }]
 dyn_wire_l_02 [{
   shapeType=Capsule
   length=3.9307
   radius=0.6501
   shapeName=dyn_wire_l_02_capsule
   center=<-4.6665 0.19 0.1089>
   rotation=<rot -0.0002 0.0002 -95.6254>
  }]
 dyn_wire_l_master [
  {
   shapeType=Capsule
   length=8.3057
   radius=2.7433
   shapeName=dyn_wire_l_master_capsule
   center=<-0.2841 0.0102 0.9857>
   rotation=<rot 25.3124 0 -0.0002>
  }
  {
   shapeType=Box
   extentX=1.8736
   extentY=4.5287
   extentZ=4.3376
   shapeName=dyn_wire_l_master_box
   center=<-3.4231 0.0067 2.9042>
   rotation=<rot 42.1248 -3.7934 -2.5468>
  }
 ]
 dyn_wire_r_01 [{
   shapeType=Capsule
   length=3.12
   radius=0.8319
   shapeName=dyn_wire_r_01_capsule
   center=<4.036 -0.003 0.0006>
   rotation=<rot 0.0001 0 -78.7502>
  }]
 dyn_wire_r_master [{
   shapeType=Capsule
   length=7.8847
   radius=2.6347
   shapeName=dyn_wire_r_master_capsule
   center=<0.3312 0.6794 0.8395>
   rotation=<rot -15.484 -16.4647 41.5046>
  }]
 pack [{shapeType=Box, extentX=15.7518, extentY=1.9873, extentZ=16.1913, shapeName=pack_box, center=<0 -6.3115 0.6134>}]
}
constraints [
 {
  bone1Name=dyn_arm_01
  bone2Name=dyn_arm_master
  swing1LimitDegrees=0
  swing1Motion=Locked
  swing2LimitDegrees=0
  swing2Motion=Locked
  twistLimitDegrees=15
 }
 {bone1Name=dyn_wire_l_01, bone2Name=dyn_wire_l_master, swing1LimitDegrees=3, swing2LimitDegrees=20, twistLimitDegrees=1}
 {bone1Name=dyn_wire_r_01, bone2Name=dyn_wire_r_master, swing1LimitDegrees=5, swing2LimitDegrees=25, twistLimitDegrees=1}
 {bone1Name=dyn_wire_l_02, bone2Name=dyn_wire_l_master, swing1LimitDegrees=3, swing2LimitDegrees=35, twistLimitDegrees=1}
]
